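# Niagara system for the idle effect of the Golden Llama supply drop, judging by its name and folder.
# It holds two emitters, Glow and Sparkles001, under a system that uses the SystemState module.
# It exposes no user variables, and no renderers, data interfaces, dynamic inputs or emitter-level modules are recorded, so the look and behavior of each emitter are not visible here.
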NiagaraSystem /BRMapCh6/Environments/SupplyDrops/FX/NS_LWE_GoldenLlama_Idle
dependencies {usedDataInterfaces [], usedDynamicInputs [], usedModules [&SystemState]}
emitters {Glow {rendererClasses [], renderers []}, Sparkles001 {rendererClasses [], renderers []}}
summary {
 systemName=NS_LWE_GoldenLlama_Idle
 emitters [{emitterName=Sparkles001, rendererClasses []}, {emitterName=Glow, rendererClasses []}]
 userVariables []
}
