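# FindSaturation is a library-exposed Math material function that estimates the saturation of an input color.
# It breaks the color into components, then divides max minus min by max plus min, with an If returning zero when that sum is zero.
# The inner Min and Max nodes both take red twice, so green is ignored and only red and blue affect the result.
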MaterialFunction /Engine/Functions/Engine_MaterialFunctions02/Math/FindSaturation
graph {
 properties {bExposeToLibrary=true, description=Finds the saturation of a color., libraryCategoriesText [Math]}
 nodes [
  n1 = MaterialFunctionCall(Float3=FunctionInput() {inputName=Color}) {materialFunction=&BreakOutFloat3Components}
  n2 = Min(A=Min(A=@n1.R, B=@n1.R), B=@n1.B)
  n3 = Max(A=Max(A=@n1.R, B=@n1.R), B=@n1.B)
  n4 = Add(A=@n2, B=@n3)
  n5 = Divide(A=Subtract(A=@n3, B=@n2), B=@n4)
  FunctionOutput(None=If(A=@n4, B=Constant(), "A > B"=@n5, "A == B"=Constant(), "A < B"=@n5)) {outputName=Saturation}
 ]
}
