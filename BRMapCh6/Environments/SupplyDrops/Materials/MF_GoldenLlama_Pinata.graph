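# Material function that gives the Golden Llama supply drop piñata a fuzzy, sparkling gold velvet look, borrowing Golden Fleece statuette textures and glitter.
# It overrides incoming base color, normal, metallic and roughness through masks, including sphere masks around the eyes.
# Tinted velvet height and grunge feed FuzzyShading and a hue shift, while velvet normals blend over another normal.
# Sparkles combine animated glitter with six sine-pulsed bursts at fixed pre-skinned positions.
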MaterialFunction /BRMapCh6/Environments/SupplyDrops/Materials/MF_GoldenLlama_Pinata
graph {nodes [
  Constant() {r=1}
  n1 = NamedRerouteUsage()
  n2 = NamedRerouteUsage()
  n3 = Constant3Vector() {constant=<rgba 5 20 40 0>}
  n4 = Constant3Vector() {constant=<rgba -5 20 40 0>}
  n5 = Add(A=NamedRerouteUsage(), B=Add(A=SphereMask(A=@n2, B=@n3) {attenuationRadius=2.5}, B=SphereMask(A=@n2, B=@n4) {attenuationRadius=2.5}))
  n6 = NamedRerouteUsage()
  n7 = Add(A=@n6) {constB=-0.2}
  n8 = Add(A=@n6) {constB=0.1}
  n9 = SetMaterialAttributes!(
   MaterialAttributes=NamedRerouteUsage()
   EmissiveColor
   Normal=LinearInterpolate(A=NamedRerouteUsage(), B=NamedRerouteUsage(), Alpha=@n1)
   BaseColor=LinearInterpolate(A=NamedRerouteUsage(), B=NamedRerouteUsage(), Alpha=Saturate(None=@n5))
   Metallic=LinearInterpolate(A=NamedRerouteUsage(), B=@n1, Alpha=NamedRerouteUsage())
   Roughness=LinearInterpolate(A=Saturate(None=@n7), B=Saturate(None=@n8), Alpha=@n1)
  ) {attributeSetTypes [
    #1
    #2
    #3
    #4
    #5
   ]}
  FunctionOutput(None=@n9)
  MaterialFunctionCall!(Color=NamedRerouteUsage()) {materialFunction=/Engine/Functions/Engine_MaterialFunctions02/Math/FindSaturation}
  n10 = TextureCoordinate()
  n11 = TextureSample(UVs=NamedRerouteDeclaration(None=Multiply(A=@n10) {constB=1.5}) {name=UVFuzz, nodeColor=<rgba 0 0.2266 1 1>}) {
   samplerSource=SSM_Wrap_WorldGroupSettings
   samplerType=SAMPLERTYPE_Normal
   texture=/Game/Environments/Helios/Props/Quest/Quest_GoldenFleeceStatuette_A/Textures/T_Velvet_N1
  }
  n12 = Constant() {r=0.8}
  NamedRerouteDeclaration(None=MaterialFunctionCall(
    BaseNormal=NamedRerouteUsage()
    AdditionalNormal=MaterialFunctionCall(Normal=@n11.RGB, Flatness=@n12) {materialFunction=&FlattenNormal}.Result
   ) {materialFunction=&BlendAngleCorrectedNormals}.Result) {name=Normal, nodeColor=<rgba 1 0 0.9999 1>}
  n13 = GetMaterialAttributes(None=FunctionInput() {inputName=MaterialAttributes, inputType=FunctionInput_MaterialAttributes}) {attributeGetTypes [
    #1
    #2
    #3
    #4
    #5
   ]}
  NamedRerouteDeclaration(None=@n13.Normal) {name=OriginalNormal, nodeColor=<rgba 1 0.6094 0 1>}
  NamedRerouteDeclaration(None=@n13.Roughness) {name=OriginalRoughness, nodeColor=<rgba 1 0 0.3984 1>}
  n14 = NamedRerouteUsage()
  n15 = SphereMask(A=NamedRerouteUsage(), B=Constant3Vector() {constant=<rgba 5 -10 12 0>}) {attenuationRadius=28, hardnessPercent=0}
  n16 = Multiply(A=@n15, B=@n15)
  n17 = Time()
  n18 = ScalarParameter() {defaultValue=0.25, group=GoldenFleece, parameterName=Bursts_Speed}
  n19 = Add(A=@n16, B=Multiply(A=Add(A=@n17) {constB=0.1}, B=@n18))
  n20 = Multiply(A=@n16, B=Saturate(None=Sine(None=@n19)))
  n21 = SphereMask(A=NamedRerouteUsage(), B=Constant3Vector() {constant=<rgba -6 -10 15 0>}) {attenuationRadius=23, hardnessPercent=0}
  n22 = Multiply(A=@n21, B=@n21)
  n23 = Add(A=Time()) {constB=0.82}
  n24 = Sine(None=Add(A=@n22, B=Multiply(A=@n23, B=@n18)))
  n25 = SphereMask(A=NamedRerouteUsage(), B=Constant3Vector() {constant=<rgba -17 -1.75 45 0>}) {attenuationRadius=23, hardnessPercent=0}
  n26 = Multiply(A=@n25, B=@n25)
  n27 = Multiply(A=Add(A=Time()) {constB=0.3}, B=@n18)
  n28 = Saturate(None=Sine(None=Add(A=@n26, B=@n27)))
  n29 = SphereMask(A=NamedRerouteUsage(), B=Constant3Vector() {constant=<rgba -13 5 20 0>}) {attenuationRadius=30, hardnessPercent=0}
  n30 = Multiply(A=@n29, B=@n29)
  n31 = Time()
  n32 = Add(A=@n30, B=Multiply(A=Add(A=@n31) {constB=0.55}, B=@n18))
  n33 = Multiply(A=@n30, B=Saturate(None=Sine(None=@n32)))
  n34 = SphereMask(A=NamedRerouteUsage(), B=Constant3Vector() {constant=<rgba 18 -8 20 0>}) {attenuationRadius=20, hardnessPercent=0}
  n35 = Multiply(A=@n34, B=@n34)
  n36 = Add(A=Time()) {constB=0.37}
  n37 = Sine(None=Add(A=@n35, B=Multiply(A=@n36, B=@n18)))
  n38 = SphereMask(A=NamedRerouteUsage(), B=Constant3Vector() {constant=<rgba 0 8 18 0>}) {attenuationRadius=34, hardnessPercent=0}
  n39 = Multiply(A=@n38, B=@n38)
  n40 = Time()
  n41 = Sine(None=Add(A=@n39, B=Multiply(A=@n40, B=@n18)))
  n42 = Add(A=Multiply(A=@n35, B=Saturate(None=@n37)), B=Multiply(A=@n39, B=Saturate(None=@n41)))
  n43 = Add(A=Multiply(A=@n22, B=Saturate(None=@n24)), B=Add(A=Multiply(A=@n26, B=@n28), B=Add(A=@n33, B=@n42)))
  n44 = Multiply(
   A=Multiply(A=Add(A=@n20, B=@n43), B=VectorParameter() {group=Override, parameterName=SparkleTint01, defaultValue=<rgba 1 0.6109 0.1818 1>}.RGB)
   B=ScalarParameter() {defaultValue=0.2, group=GoldenFleece, parameterName=Bursts_Intensity, sortPriority=11}
  ) {constB=0.2}
  n45 = MaterialFunctionCall() {materialFunction=/Game/Environments/Helios/Props/Quest/Quest_GoldenFleeceStatuette_A/Materials/MF_GlitterFX_GoldenFleeceStatuette}
  n46 = VectorParameter() {group=Override, parameterName=SparkleTint02, defaultValue=<rgba 0.2865 0.158 0.0164 1>}
  n47 = Multiply(
   A=Multiply(
    A=Multiply(A=@n45.GlitterAnimated, B=@n46.RGB)
    B=ScalarParameter() {defaultValue=3, group=GoldenFleece, parameterName=Sparkles_Intensity, sortPriority=11}
   ) {constB=3}
   B=OneMinus(None=NamedRerouteUsage())
  )
  NamedRerouteDeclaration(None=Multiply(A=OneMinus(None=@n14), B=Add(A=@n44, B=@n47))) {name=Sparkles, nodeColor=<rgba 1 0.1875 0 1>}
  n48 = NamedRerouteUsage()
  n49 = Constant()
  NamedRerouteDeclaration(None=OneMinus(None=SphereMask(A=@n48, B=@n49) {attenuationRadius=0.3, hardnessPercent=42})) {name=EyesMask, nodeColor=<rgba 0 1 0.5313 1>}
  NamedRerouteDeclaration(None=Constant()) {name=SeattleMask, nodeColor=<rgba 0.2422 1 0 1>}
  n50 = PreSkinnedPosition()
  NamedRerouteDeclaration(None=ShaderStageSwitch(PixelShader=VertexInterpolator(VS=@n50).PS, VertexShader=@n50)) {name=PreskinnedPosition, nodeColor=<rgba 1 0 0.3281 1>}
  n51 = VectorParameter() {group=Override, parameterName=ColorTint02, defaultValue=<rgba 0.9635 0.7199 0.2007 1>}
  n52 = TextureSample(UVs=NamedRerouteUsage()) {
   samplerSource=SSM_Wrap_WorldGroupSettings
   samplerType=SAMPLERTYPE_Masks
   texture=/Game/Environments/Helios/Props/Quest/Quest_GoldenFleeceStatuette_A/Textures/T_Velvet_Height1
  }
  n53 = Multiply(A=Saturate(None=Add(A=@n52.RGB) {constB=0.15}), B=Multiply(A=NamedRerouteUsage()) {constB=3.5})
  n54 = TextureSample(UVs=Multiply(A=TextureCoordinate()) {constB=2}) {
   samplerSource=SSM_Wrap_WorldGroupSettings
   samplerType=SAMPLERTYPE_Masks
   texture=/Game/Environments/Helios/Props/Quest/Quest_GoldenFleeceStatuette_A/Textures/T_Velvet_GrungeMask1
  }
  n55 = Multiply(A=Multiply(A=@n54.R, B=VectorParameter() {group=Override, parameterName=ColorTint01, defaultValue=<rgba 1 0.6696 0.1583 1>}.RGB)) {constB=0.7}
  n56 = MaterialFunctionCall!(
   BaseColor=LinearInterpolate(A=Multiply(A=@n51.RGB) {constB=0.75}, B=Add(A=@n53, B=@n55), Alpha=NamedRerouteUsage())
   Normal=NamedRerouteUsage()
   CoreDarkness=Constant() {r=0.7}
   Power=Constant() {r=1}
   EdgeBrightness=Constant() {r=0.2}
  ) {materialFunction=/Engine/Functions/Engine_MaterialFunctions01/Shading/FuzzyShading}
  n57 = Constant() {r=-0.25}
  n58 = NamedRerouteUsage()
  n59 = Constant3Vector() {constant=<rgba 0 20 40 0>}
  n60 = Multiply(
   A=MaterialFunctionCall(Hue Shift Percentage=Constant() {r=-0.02}, Texture=Desaturation(None=@n56.Result, Fraction=@n57)) {materialFunction=&HueShift}.Result
   B=LinearInterpolate(Alpha=SphereMask(A=@n58, B=@n59) {attenuationRadius=18, hardnessPercent=0.1}) {constA=1, constB=2}
  )
  n61 = NamedRerouteUsage()
  NamedRerouteDeclaration(None=LinearInterpolate(A=NamedRerouteUsage(), B=Add(A=@n60, B=@n61), Alpha=NamedRerouteUsage())) {name=BaseColor, nodeColor=<rgba 1 0.8672 0 1>}
  NamedRerouteDeclaration(None=@n13.BaseColor) {name=OriginalBaseColor, nodeColor=<rgba 1 0.1406 0 1>}
  NamedRerouteDeclaration(None=@n13.MaterialAttributes) {name=OriginalMA, nodeColor=<rgba 1 0 0.1015 1>}
  NamedRerouteDeclaration(None=@n13.Metallic) {name=OriginalMetallic, nodeColor=<rgba 0 0.8594 1 1>}
  n62 = Fresnel(Normal=NamedRerouteUsage()) {exponent=6}
  NamedRerouteDeclaration(None=Multiply(A=LinearInterpolate(Alpha=@n62) {constA=0.8, constB=0.6}, B=NamedRerouteUsage())) {name=Metallic, nodeColor=<rgba 0 0.2266 1 1>}
  NamedRerouteUsage()
  NamedRerouteUsage()
 ]}
